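# Unlit additive material for Niagara mesh particles that renders the engine glow on the female Motorcyclist glider.
# The inverted red channel of the glider's diffuse texture is raised to a power, tinted by the Glow color and passed through an emissive-adjustment function.
# A particle-driven Heat value scales a second-UV gradient and camera-facing falloff, sharpened by an exponent.
# A Push parameter scales the vertex normal to form an offset vector.
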Material /BRCosmetics/Gadgets/Assets/VinderTech_GliderChute/Glider_Motorcyclist_Female/FX/Materials/M_Glider_Motorcyclist_Female_EngineGlow
graph {
 properties {
  bUsedWithNiagaraMeshParticles=true
  blendMode=BLEND_Additive
  pixelDepthOffsetMode=PDOM_Legacy
  shadingModel=MSM_Unlit
  usedShadingModels=Unlit
 }
 nodes [
  n1 = OneMinus(None=TextureSample() {texture=/BRCosmetics/Gadgets/Assets/VinderTech_GliderChute/Glider_Motorcyclist_Female/Textures/T_Glider_Motorcyclist_Female_D}.R)
  n2 = VectorParameter() {parameterName=Glow, defaultValue=<rgba 6 1.1397 0.24 3>}
  MaterialFunctionCall(Emissive=Multiply(A=Power(Base=@n1, Exp=@n2.A), B=@n2.RGB)) {materialFunction=&MF_FX_EmissiveAdjustments}
  n3 = VertexNormalWS()
  Multiply(A=@n3, B=ScalarParameter() {defaultValue=0.4, parameterName=Push})
  n4 = ComponentMask(None=TextureCoordinate() {coordinateIndex=1}) {g=true}
  n5 = OneMinus(None=Saturate(None=Abs(None=@n4)))
  n6 = DotProduct(A=CameraVectorWS(), B=@n3)
  Power(
   Base=Multiply(A=DynamicParameter() {parameterIndex=1, paramNames [Heat, "", "", ""]}.Heat, B=Multiply(A=@n5, B=@n6))
   Exp=ScalarParameter() {defaultValue=3, parameterName=Glow Exponent}
  )
 ]
}
