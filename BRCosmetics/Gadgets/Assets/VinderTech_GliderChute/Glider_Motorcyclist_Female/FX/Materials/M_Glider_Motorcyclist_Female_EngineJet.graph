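# Additive, unlit, two-sided Niagara mesh-particle material for the engine jet on the female Motorcyclist glider.
# Offset math combines a V-gradient Z push with a push along normals.
# Emissive blends particle color with a dynamic secondary color built only from its R channel, via MF_FX_EmissiveAdjustments.
# Opacity multiplies particle alpha, a camera-facing term that softens edges, and a V gradient whose exponent flickers through stepped, per-particle time noise.
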Material /BRCosmetics/Gadgets/Assets/VinderTech_GliderChute/Glider_Motorcyclist_Female/FX/Materials/M_Glider_Motorcyclist_Female_EngineJet
graph {
 properties {
  bUsedWithNiagaraMeshParticles=true
  blendMode=BLEND_Additive
  pixelDepthOffsetMode=PDOM_Legacy
  shadingModel=MSM_Unlit
  twoSided=true
  usedShadingModels=Unlit
 }
 nodes [
  n1 = ComponentMask(None=TextureCoordinate()) {g=true}
  n2 = OneMinus(None=@n1)
  n3 = ScalarParameter() {defaultValue=-2, parameterName=DownPush}
  n4 = VertexNormalWS()
  Add(
   A=Multiply(A=Multiply(A=@n2, B=@n3), B=Constant3Vector() {constant=<rgba 0 0 1 0>})
   B=Multiply(A=@n4, B=ScalarParameter() {defaultValue=-0.5, parameterName=Push})
  )
  n5 = DynamicParameter() {paramNames [Secondary Color R, Secondary Color G, Secondary Color B, Color Blend]}
  n6 = ParticleColor()
  MaterialFunctionCall(Emissive=LinearInterpolate(
    A=MaterialFunctionCall(R=@n5.Secondary Color R, G=@n5.Secondary Color R, B=@n5.Secondary Color R) {materialFunction=&AppendMany}.RGB
    B=@n6.RGB
    Alpha=@n5.Color Blend
   )) {materialFunction=&MF_FX_EmissiveAdjustments}
  n7 = Add(A=Time(), B=ParticleRandom())
  n8 = ScalarParameter() {defaultValue=266, parameterName=Flicker Speed}
  n9 = Add(A=Floor(None=Multiply(A=@n7, B=@n8))) {constB=0.5}
  n10 = LinearInterpolate(Alpha=ComponentMask(None=VectorNoise(World Position=@n9)) {r=true}) {constA=3, constB=6}
  n11 = Abs(None=DotProduct(A=@n4, B=CameraVectorWS()))
  n12 = ScalarParameter() {defaultValue=4, parameterName=Fresnel Exponent}
  Multiply(A=@n6.A, B=Multiply(A=Power(Base=@n1, Exp=@n10), B=Power(Base=@n11, Exp=@n12)))
 ]
}
