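# Additive, unlit, two-sided particle material for the exhaust effect on the female Motorcyclist glider, usable on Niagara sprites and ribbons.
# Dynamic parameters blend particle color toward a secondary color, which reads only the R channel for all three components and so is grayscale, then feed MF_FX_EmissiveAdjustments.
# A hard circular UV mask is multiplied by particle alpha and a depth fade, which mobile shading replaces with a constant one.
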Material /BRCosmetics/Gadgets/Assets/VinderTech_GliderChute/Glider_Motorcyclist_Female/FX/Materials/M_Glider_Motorcyclist_Female_Exhaust
graph {
 properties {
  bUsedWithNiagaraRibbons=true
  bUsedWithNiagaraSprites=true
  blendMode=BLEND_Additive
  pixelDepthOffsetMode=PDOM_Legacy
  shadingModel=MSM_Unlit
  twoSided=true
  usedShadingModels=Unlit
 }
 nodes [
  n1 = DynamicParameter() {paramNames [Secondary Color R, Secondary Color G, Secondary Color B, Color Blend]}
  n2 = ParticleColor()
  MaterialFunctionCall(Emissive=LinearInterpolate(
    A=MaterialFunctionCall(R=@n1.Secondary Color R, G=@n1.Secondary Color R, B=@n1.Secondary Color R) {materialFunction=&AppendMany}.RGB
    B=@n2.RGB
    Alpha=@n1.Color Blend
   )) {materialFunction=&MF_FX_EmissiveAdjustments}
  n3 = Constant() {r=0.5}
  n4 = DepthFade(FadeDistance=ScalarParameter() {defaultValue=10, parameterName=Depth Fade Distance})
  n5 = Constant() {r=1}
  Multiply(
   A=SphereMask(A=TextureCoordinate(), B=@n3, Radius=@n3, Hardness=Constant() {r=1})
   B=Multiply(A=@n2.A, B=ShadingPathSwitch(Default=@n4, Mobile=@n5) {desc=Migrated from FeatureLevelSwitch})
  )
  OneMinus(None=@n4)
 ]
}
