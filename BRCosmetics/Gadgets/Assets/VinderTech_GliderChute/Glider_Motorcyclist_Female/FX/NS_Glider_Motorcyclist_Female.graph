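# Niagara effect for the Motorcyclist Female glider: engine heat meshes, exhaust glow sprites, exhaust ribbons and anim-trail wing ribbons.
# Heat, fill and exhaust particles follow bones or sockets on User.SkeletalMesh, split by spawn group.
# User.EmissiveColorSelection indexes per-emitter color arrays, with secondary color passed through dynamic material parameters.
# System timer values drive alpha and material parameters; ribbon alpha also scales by time of day, and a debug emitter is disabled.
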NiagaraSystem /BRCosmetics/Gadgets/Assets/VinderTech_GliderChute/Glider_Motorcyclist_Female/FX/NS_Glider_Motorcyclist_Female
dependencies {
 usedDataInterfaces [&NiagaraDataInterfaceArrayColor, &NiagaraDataInterfaceCurve, &NiagaraDataInterfaceColorCurve]
 usedDynamicInputs [&FN_FrontEndFloatModifier]
 usedModules [&SystemState, &FN_Integrator]
}
emitters {
 Engine_Heat {
  emitterUpdateScript [
   &EmitterState(
    Inactive Response=::NewEnumerator1
    Life Cycle Mode=::NewEnumerator1
    Loop Behavior=::NewEnumerator1
    Loop Duration Mode=::NewEnumerator1
    UseLoopDelay=false
   )
   &SpawnBurst_Instantaneous(Spawn Count=3, Use Spawn Probability=false, UseLoopCountLimit=false)
  ]
  particleSpawnScript [
   &InitializeParticle(
    Color=&SelectLinearColorFromArray(
     Array Sampling Mode=::NewEnumerator1
     Direct Array Index=&MakeIntFromFloat(Float=$User.EmissiveColorSelection)
     Color Selection Array=DataInterface() {propertyValues {ColorData [<rgba 0 11 11 1>, <rgba 22 0 5.4999 1>, <rgba 11 6.9667 0 1>, <rgba 19.8 22 22 1>]}}
     Random Seed=?
    )
    Lifetime=0
    UsePositionOffset=false
    Write Material Random=true
   )
   &FN_SetSpawnGroup(Number of Spawn Groups=3)
   /SetVariables(Particles.Secondary Color:LinearColor=&SelectLinearColorFromArray(
     Array Sampling Mode=::NewEnumerator1
     Direct Array Index=&MakeIntFromFloat(Float=$User.EmissiveColorSelection)
     Color Selection Array=DataInterface() {propertyValues {ColorData [<rgba 9 0 9 1>, <rgba 9 1.8 9 1>, <rgba 9 0 9 1>, <rgba 9 1.35 9 1>]}}
     Random Seed=?
    )) {isSetParameters=true}
   &DynamicMaterialParameters(
    Index 0 Param 1=&MakeFloatFromLinearColor(LinearColor=$Particles.Secondary Color, Channel:ENiagaraLinearColor_Channels=::NewEnumerator0)
    Index 0 Param 2=&MakeFloatFromLinearColor(LinearColor=$Particles.Secondary Color, Channel:ENiagaraLinearColor_Channels=::NewEnumerator1)
    Index 0 Param 3=&MakeFloatFromLinearColor(LinearColor=$Particles.Secondary Color, Channel:ENiagaraLinearColor_Channels=::NewEnumerator2)
    Module.Param0WriteEnabled=true
    Module.Param1WriteEnabled=true
    Module.Param2WriteEnabled=true
    Module.Param3WriteEnabled=true
   ) {module=DynamicMaterialParameters001}
  ]
  particleUpdateScript [
   &DynamicMaterialParameters(
    Index 0 Param 4=$System.FN_Timer001.Value
    Index 1 Param 1=$System.FN_Timer002.Value
    Module.Param0WriteEnabled=false
    Module.Param0WriteEnabled1=true
    Module.Param1WriteEnabled=false
    Module.Param1WriteEnabled1=false
    Module.Param2WriteEnabled=false
    Module.Param2WriteEnabled1=false
    Module.Param3WriteEnabled=true
    Module.Param3WriteEnabled1=false
    Write Parameter Index 1=true
   )
   &ParticleState(Kill Particles When Lifetime Has Elapsed=false, Let Infinitely Lived Particles Die When Emitter Deactivates=true)
   &SkeletalMeshLocation(Bone / Socket Index=$Particles.SpawnGroup, Bone Sampling Mode=::NewEnumerator4, MaskSpawnGroup=false, Skeletal Mesh=$User.SkeletalMesh)
   &ScaleColor(Scale Alpha=&Multiply_Float(A=1, B=$System.FN_Timer001.Value), ScaleA=true, ScaleRGB=false)
  ]
  rendererClasses [&NiagaraMeshRendererProperties]
  renderers [{rendererClass=&NiagaraMeshRendererProperties, rendererIndex=0}]
 }
 Exhaust_Fill {
  emitterUpdateScript [
   &EmitterState(
    Inactive Response=::NewEnumerator1
    Life Cycle Mode=::NewEnumerator1
    Loop Behavior=::NewEnumerator1
    Loop Duration Mode=::NewEnumerator1
    UseLoopDelay=false
   )
   &SpawnBurst_Instantaneous(Spawn Count=2, Use Spawn Probability=false, UseLoopCountLimit=false)
  ]
  particleSpawnScript [
   &FN_SetSpawnGroup(Number of Spawn Groups=2)
   &InitializeParticle(
    Color=&SelectLinearColorFromArray(
     Array Sampling Mode=::NewEnumerator1
     Direct Array Index=&MakeIntFromFloat(Float=$User.EmissiveColorSelection)
     Color Selection Array=DataInterface() {propertyValues {ColorData [<rgba 0 110.9999 111 1>, <rgba 222 0 55.4996 1>, <rgba 111 70.3 0 1>, <rgba 199.8 222 222 1>]}}
     Random Seed=?
    )
    Lifetime=0
    Sprite Size Mode=::NewEnumerator3
    Uniform Sprite Size=7
    UsePositionOffset=false
    Write Material Random=true
   )
   &SkeletalMeshLocation(
    Bone / Socket Index=&MaskIntBySpawnGroup(Masked Int=4, Passthrough Int=5, Spawn Group Mask=0)
    MaskSpawnGroup=false
    Mesh Sampling Type=::NewEnumerator5
    Orientation Sampling=::NewEnumerator1
    Sampled Position Offset Coordinate Space:ENiagaraSkelMeshTransforms=::NewEnumerator2
    Skeletal Mesh=$User.SkeletalMesh
    Socket Sampling Mode=::NewEnumerator5
    Use Sampled Mesh Attribute as Ribbon ID=false
   )
   &SpriteFacingAndAlignment(Facing Coordinate Space=::Local, Sprite Facing=<0 1 0.3>, WriteAlignment=false, WriteFacing=true)
   /SetVariables(Particles.Secondary Color:LinearColor=&SelectLinearColorFromArray(
     Array Sampling Mode=::NewEnumerator1
     Direct Array Index=&MakeIntFromFloat(Float=$User.EmissiveColorSelection)
     Color Selection Array=DataInterface() {propertyValues {ColorData [<rgba 9 0 9 1>, <rgba 9 1.8 9 1>, <rgba 9 0 9 1>, <rgba 9 1.35 9 1>]}}
     Random Seed=?
    )) {isSetParameters=true}
   &DynamicMaterialParameters(
    Index 0 Param 1=&MakeFloatFromLinearColor(LinearColor=$Particles.Secondary Color, Channel:ENiagaraLinearColor_Channels=::NewEnumerator0)
    Index 0 Param 2=&MakeFloatFromLinearColor(LinearColor=$Particles.Secondary Color, Channel:ENiagaraLinearColor_Channels=::NewEnumerator1)
    Index 0 Param 3=&MakeFloatFromLinearColor(LinearColor=$Particles.Secondary Color, Channel:ENiagaraLinearColor_Channels=::NewEnumerator2)
    Module.Param0WriteEnabled=true
    Module.Param1WriteEnabled=true
    Module.Param2WriteEnabled=true
    Module.Param3WriteEnabled=true
   ) {module=DynamicMaterialParameters001}
  ]
  particleUpdateScript [
   &DynamicMaterialParameters(
    Index 0 Param 4=$System.FN_Timer001.Value
    Module.Param0WriteEnabled=false
    Module.Param1WriteEnabled=false
    Module.Param2WriteEnabled=false
    Module.Param3WriteEnabled=true
   )
   &SkeletalMeshLocation(
    Bone / Socket Index=&MaskIntBySpawnGroup(Masked Int=4, Passthrough Int=5, Spawn Group Mask=0)
    MaskSpawnGroup=false
    Mesh Sampling Type=::NewEnumerator5
    Sampled Position Offset Coordinate Space:ENiagaraSkelMeshTransforms=::NewEnumerator2
    Skeletal Mesh=$User.SkeletalMesh
    Socket Sampling Mode=::NewEnumerator5
   ) {module=SkeletalMeshLocation001}
   &ParticleState(Kill Particles When Lifetime Has Elapsed=false, Let Infinitely Lived Particles Die When Emitter Deactivates=true)
   &ScaleColor(Scale Alpha=&Multiply_Float(A=0.3, B=$System.FN_Timer001.Value), ScaleA=true, ScaleRGB=false)
  ]
  rendererClasses [&NiagaraSpriteRendererProperties]
  renderers [{rendererClass=&NiagaraSpriteRendererProperties}]
 }
 Exhaust_Trails {
  emitterUpdateScript [
   &EmitterState()
   &FN_SpawnRibbon_v2(Number of Ribbons=2, SpawnRate=350)
  ]
  particleSpawnScript [
   &FN_InitializeRibbon_v3(
    Color:LinearColor=&SelectLinearColorFromArray(
     Array Sampling Mode=::NewEnumerator1
     Direct Array Index=&MakeIntFromFloat(Float=$User.EmissiveColorSelection)
     Color Selection Array=DataInterface() {propertyValues {ColorData [<rgba 0 9 9 1>, <rgba 9 0 2.25 1>, <rgba 9 5.7 0 1>, <rgba 8.1 9 9 1>]}}
     Random Seed=?
    )
    Custom Offset=true
    Enable Skeletal Position=true
    Facing Axis:ENiagaraOrientationAxis=::YAxis
    Lifetime=0.075
    Local Vector Offset:Vector3f=<0 1 0>
    Ribbon Width=8
   )
   &ApplyInitialForces()
   &FN_UpdateRibbon(Detach Locked Particle=true, Enable Skeletal Position=true)
   &FN_SetSpawnGroup(Number of Spawn Groups=2)
   &SkeletalMeshLocation(
    Bone / Socket Index=&MaskIntBySpawnGroup(Masked Int=4, Passthrough Int=5, Spawn Group Mask=0)
    MaskSpawnGroup=false
    Mesh Sampling Type=::NewEnumerator5
    Orientation Sampling=::NewEnumerator1
    Sampled Position Offset=<0 1 0>
    Sampled Position Offset Coordinate Space:ENiagaraSkelMeshTransforms=::NewEnumerator2
    Skeletal Mesh=$User.SkeletalMesh
    Socket Sampling Mode=::NewEnumerator5
   )
   &AddVelocity(Velocity=&FN_FrontEndVector3Modifier(Gameplay=<0 3333 0>, Front End=<0 3333 0>))
   /SetVariables(
    Particles.Intensity Multiplier:NiagaraFloat=&Multiply_Float(A=0.26, B=$System.FN_Timer001.Value)
    Particles.Secondary Color:LinearColor=&SelectLinearColorFromArray(
     Array Sampling Mode=::NewEnumerator1
     Direct Array Index=&MakeIntFromFloat(Float=$User.EmissiveColorSelection)
     Color Selection Array=DataInterface() {propertyValues {ColorData [<rgba 9 0 9 1>, <rgba 9 1.8 9 1>, <rgba 9 0 9 1>, <rgba 9 1.35 9 1>]}}
     Random Seed=?
    )
   ) {isSetParameters=true, module=SetVariables_#1}
   /SetVariables(
    Particles.Color=&MultiplyLinearColor(
     Linear Color=$StackContext.Color
     Scale Factor=&MakeVector4_FromFloat(Float=$StackContext.Intensity Multiplier)
    )
    Particles.Secondary Color:LinearColor=&MultiplyLinearColor(
     Linear Color=$StackContext.Secondary Color
     Scale Factor=&MakeVector4_FromFloat(Float=$StackContext.Intensity Multiplier)
    )
   ) {isSetParameters=true}
   &DynamicMaterialParameters(
    Index 0 Param 1=&MakeFloatFromLinearColor(LinearColor=$Particles.Secondary Color, Channel:ENiagaraLinearColor_Channels=::NewEnumerator0)
    Index 0 Param 2=&MakeFloatFromLinearColor(LinearColor=$Particles.Secondary Color, Channel:ENiagaraLinearColor_Channels=::NewEnumerator1)
    Index 0 Param 3=&MakeFloatFromLinearColor(LinearColor=$Particles.Secondary Color, Channel:ENiagaraLinearColor_Channels=::NewEnumerator2)
    Index 0 Param 4=&Multiply_Float(A=$System.FN_Timer.Value, B=2)
    Index 1 Param 2=1
    Index 1 Param 4=&MaskFloatBySpawnGroup(Masked Float=0, Passthrough Float=0.5)
    Index 2 Param 1=1
    Module.Param0WriteEnabled=true
    Module.Param0WriteEnabled1=true
    Module.Param0WriteEnabled2=true
    Module.Param1WriteEnabled=true
    Module.Param1WriteEnabled1=true
    Module.Param1WriteEnabled2=false
    Module.Param2WriteEnabled=true
    Module.Param2WriteEnabled1=true
    Module.Param2WriteEnabled2=false
    Module.Param3WriteEnabled=true
    Module.Param3WriteEnabled1=true
    Module.Param3WriteEnabled2=false
    Write Parameter Index 1=true
    Write Parameter Index 2=true
   ) {module=DynamicMaterialParameters001}
  ]
  particleUpdateScript [
   &InheritSourceMovement(
    Applied Position Scale=&Multiply_VectorByFloat(
     Vector=<1 1 1>
     Float=&FloatFromCurve(FloatCurve=DataInterface() {propertyValues {Curve {keys [{tangentMode=RCTM_User, time=0, value=1}, {tangentMode=RCTM_User, time=1, value=0}]}}})
    )
    ApplyForce=false
    ApplyPosition=true
    ApplyPositionFalloff=false
   )
   &Drag(Drag=33, Use Linear Drag=true, Use Rotational Drag=false)
   &SolveForcesAndVelocity(Clamp Velocity=false, Limit Acceleration=false)
   &FN_UpdateRibbon(Detach Locked Particle=true) {module=FN_UpdateRibbon001}
   &ScaleColor(
    Color Value To Scale=$Particles.Color
    Scale Alpha=&Multiply_Float(
     A=$System.FN_Timer001.Value
     B=&FNTimeOfDayScalar(Night Scale=0.3, Morning Scale=0.5, Evening Scale=0.5)
    )
    ScaleA=true
    ScaleRGB=false
   ) {module=ScaleColor001}
   &DynamicMaterialParameters(
    Index 0 Param 4=&Multiply_Float(A=$System.FN_Timer.Value, B=2)
    Module.Param0WriteEnabled=false
    Module.Param1WriteEnabled=false
    Module.Param2WriteEnabled=false
    Module.Param3WriteEnabled=true
   )
  ]
  rendererClasses [&NiagaraRibbonRendererProperties]
  renderers [{rendererClass=&NiagaraRibbonRendererProperties}]
 }
 FN_Debug {
  bEnabled=false
  emitterUpdateScript [
   &EmitterState()
   &SpawnBurst_Instantaneous(Spawn Count=1, Use Spawn Probability=false)
   &SpawnRate(SpawnRate=3, Use Spawn Probability=false)
  ]
  particleSpawnScript [/Niagara/Modules/Spawn/Initialization/InitializeParticle(
    Lifetime=5
    Module.Write Color=true
    Module.Write Lifetime=true
    Module.Write Mass=true
    Module.Write Material Random=true
    Module.Write Position=false
    Module.Write Scale=false
    Module.Write SpriteRotation=false
    Module.Write SpriteSize=true
    Position=&SimulationPosition()
    Sprite Size=<100 100>
   )]
  particleUpdateScript [
   &ParticleState()
   &FN_Timer()
   /SetVariables(Particles.Debug:NiagaraFloat=$System.FN_Timer002.Value) {isSetParameters=true}
   &DynamicMaterialParameters(
    Index 0 Param 1=$Particles.Debug
    Module.Param0WriteEnabled=true
    Module.Param1WriteEnabled=true
    Module.Param2WriteEnabled=true
    Module.Param3WriteEnabled=true
   )
   &CameraOffset(Camera Offset Amount=111)
  ]
  rendererClasses [&NiagaraSpriteRendererProperties]
  renderers [{rendererClass=&NiagaraSpriteRendererProperties}]
 }
 Wing_Trails {
  emitterUpdateScript [
   &EmitterState()
   /Game/Effects/Niagara/Modules/AnimTrail/FN_SpawnAnimTrail/FN_SpawnAnimTrail(Number of Anim Trails=2, SpawnRate=256)
  ]
  particleSpawnScript [
   /Game/Effects/Niagara/Modules/AnimTrail/FN_InitializeAnimTrail/FN_InitializeAnimTrailBoneIndecies(First Bone or Socket Index=1)
   &FN_InitializeAnimTrailBoneIndecies(AnimTrail Selection=1, First Bone or Socket Index=3, Second Bone or Socket Index=2) {module=FN_InitializeAnimTrailBoneIndecies001}
   /Game/Effects/Niagara/Modules/AnimTrail/FN_InitializeAnimTrail/FN_InitializeAnimTrail(
    AnimTrailWidth=1
    Color=&SelectLinearColorFromArray(
     Array Sampling Mode=::NewEnumerator1
     Direct Array Index=&MakeIntFromFloat(Float=$User.EmissiveColorSelection)
     Color Selection Array=DataInterface() {propertyValues {ColorData [<rgba 0 9 9 0.05>, <rgba 9 0 2.25 0.05>, <rgba 9 5.7 0 0.05>, <rgba 8.1 9 9 0.05>]}}
     Random Seed=?
    )
    Lifetime=0.13
   )
   &AddVelocity(Velocity=&FN_FrontEndVector3Modifier(Gameplay=<0 666 0>, Front End=<0 1500 0>))
   /SetVariables(Particles.Secondary Color:LinearColor=&SelectLinearColorFromArray(
     Array Sampling Mode=::NewEnumerator1
     Direct Array Index=&MakeIntFromFloat(Float=$User.EmissiveColorSelection)
     Color Selection Array=DataInterface() {propertyValues {ColorData [<rgba 9 0 9 1>, <rgba 9 1.8 9 1>, <rgba 9 0 9 1>, <rgba 9 1.35 9 1>]}}
     Random Seed=?
    )) {isSetParameters=true, module=SetVariables_#1}
   &DynamicMaterialParameters(
    Index 0 Param 1=&MakeFloatFromLinearColor(LinearColor=$Particles.Secondary Color, Channel:ENiagaraLinearColor_Channels=::NewEnumerator0)
    Index 0 Param 2=&MakeFloatFromLinearColor(LinearColor=$Particles.Secondary Color, Channel:ENiagaraLinearColor_Channels=::NewEnumerator1)
    Index 0 Param 3=&MakeFloatFromLinearColor(LinearColor=$Particles.Secondary Color, Channel:ENiagaraLinearColor_Channels=::NewEnumerator2)
    Index 1 Param 1=1
    Index 1 Param 2=0.01
    Index 1 Param 3=0.1
    Index 1 Param 4=&MaskFloatBySpawnGroup(Masked Float=0, Passthrough Float=0.5)
    Module.Param0WriteEnabled=true
    Module.Param0WriteEnabled1=true
    Module.Param0WriteEnabled2=true
    Module.Param1WriteEnabled=true
    Module.Param1WriteEnabled1=true
    Module.Param1WriteEnabled2=true
    Module.Param2WriteEnabled=true
    Module.Param2WriteEnabled1=true
    Module.Param2WriteEnabled2=true
    Module.Param3WriteEnabled=false
    Module.Param3WriteEnabled1=true
    Module.Param3WriteEnabled2=true
    Write Parameter Index 1=true
    Write Parameter Index 2=true
   ) {module=DynamicMaterialParameters001}
  ]
  particleUpdateScript [
   &InheritSourceMovement(
    Applied Position Scale=&Multiply_VectorByFloat(Vector=<1 1 1>, Float=&OneMinusFloat(Float=$System.FN_Timer001.Value))
    ApplyForce=false
    ApplyPosition=true
    ApplyPositionFalloff=false
   )
   &SolveForcesAndVelocity(Clamp Velocity=false, Limit Acceleration=false)
   /Game/Effects/Niagara/Modules/AnimTrail/FN_UpdateAnimTrail/FN_UpdateAnimTrail(New float=0.8)
   &ScaleColor(
    Color Value To Scale=$Particles.Color
    Scale Alpha=&Multiply_Float(
     A=$System.FN_Timer001.Value
     B=&FNTimeOfDayScalar(Night Scale=0.3, Morning Scale=0.5, Evening Scale=0.5)
    )
    ScaleA=true
    ScaleRGB=false
   ) {module=ScaleColor001}
   &DynamicMaterialParameters(
    Index 0 Param 4=$System.FN_Timer.Value
    Module.Param0WriteEnabled=false
    Module.Param1WriteEnabled=false
    Module.Param2WriteEnabled=false
    Module.Param3WriteEnabled=true
   )
  ]
  rendererClasses [&NiagaraRibbonRendererProperties]
  renderers [{rendererClass=&NiagaraRibbonRendererProperties}]
 }
}
summary {
 systemName=NS_Glider_Motorcyclist_Female
 emitters [
  {bEnabled=false, emitterName=FN_Debug, rendererClasses [&NiagaraSpriteRendererProperties]}
  {emitterName=Wing_Trails, rendererClasses [&NiagaraRibbonRendererProperties]}
  {emitterName=Exhaust_Trails, rendererClasses [&NiagaraRibbonRendererProperties]}
  {emitterName=Exhaust_Fill, rendererClasses [&NiagaraSpriteRendererProperties]}
  {emitterName=Engine_Heat, rendererClasses [&NiagaraMeshRendererProperties]}
 ]
 userVariables [
  User.bIsFrontEnd {type=NiagaraBool}
  User.bIsFrontEndPreview {type=NiagaraBool}
  User.bIsFullyDeployed {type=NiagaraBool}
  User.EmissiveColorSelection {type=NiagaraFloat, defaultValue=0, description=""}
  User.ForwardDot {type=NiagaraFloat, defaultValue=0, description=""}
  User.RibbonIDOffset {type=NiagaraInt32}
  User.RightDot {type=NiagaraFloat, defaultValue=0, description=""}
  User.SkeletalMesh {
   type=NiagaraDataInterfaceSkeletalMesh
   defaultValue {
    struct=NiagaraExt_VariableValue_DataInterface
    value {dataInterface=&self:NiagaraDataInterfaceSkeletalMesh_0}
   }
  }
  User.TrailColor {type=LinearColor, description="", defaultValue {struct=LinearColor, value=<rgba 1 1 1 1>}}
 ]
}
